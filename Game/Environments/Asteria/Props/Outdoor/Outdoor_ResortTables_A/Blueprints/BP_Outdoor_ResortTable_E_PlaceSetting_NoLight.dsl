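; Resort outdoor table prop with place setting and instanced patio chairs, named as the no-light variant.
; The construction script seeds a random stream from the actor's location, then offsets each chair instance's yaw within plus or minus RotationAmount for repeatable variation.
; Each offset combines with the existing rotation, also re-adding its roll and pitch.
; BeginPlay, Tick and overlap events are empty; Destroyed destroys the Box1 component.
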(event EventBeginPlay)

(event Collision|EventActorBeginOverlap
  (params (OtherActor :type Actor%20Object%20Reference)))

(event EventTick
  (params (DeltaSeconds :type Float%20%28single-precision%29)))

(event EventDestroyed
  (Components|DestroyComponent (Variables|Default|GetBox1)))

(fn ConstructionScript
  (Variables|Default|SetInstancingRandomStream (Math|Random|MakeRandomStreamfromLocation (Transformation|GetActorLocation) 1.0 true))
  (for _index (range (Components|InstancedStaticMesh|GetInstanceCount (Variables|Default|GetISM_PatioChairs)))
    (Components|InstancedStaticMesh|UpdateInstanceTransform (Variables|Default|GetISM_PatioChairs) _index (Math|Transform|MakeTransform (.location (Components|InstancedStaticMesh|GetInstanceTransform (Variables|Default|GetISM_PatioChairs) _index false)) (Math|Rotator|CombineRotators (.rotation (Components|InstancedStaticMesh|GetInstanceTransform (Variables|Default|GetISM_PatioChairs) _index false)) (Math|Rotator|MakeRotator (.roll (.rotation (Components|InstancedStaticMesh|GetInstanceTransform (Variables|Default|GetISM_PatioChairs) _index false))) (.pitch (.rotation (Components|InstancedStaticMesh|GetInstanceTransform (Variables|Default|GetISM_PatioChairs) _index false))) (Math|Random|RandomFloatinRangefromStream (Variables|Default|GetInstancingRandomStream) (- (Variables|Default|GetRotationAmount) (* (Variables|Default|GetRotationAmount) 2.0)) (Variables|Default|GetRotationAmount)))) (.scale (Components|InstancedStaticMesh|GetInstanceTransform (Variables|Default|GetISM_PatioChairs) _index false))) false false false)))
